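# Physics asset for the VelociTeeth female character's FaceAcc mesh part, covering two braids, a tail and scarf pieces.
# Five-segment braid chains hang from neck_01 and a three-segment tail from base_tail, using narrow swing and near-locked twist limits.
# Head, spine, clavicle, upper-arm and front-scarf shapes supply body volumes, with extra boxes on the right upper arm.
# Back scarf, head, clavicles and upper arms link to spine_05 with fully free angular motion.
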PhysicsAsset /BRCosmetics/Characters/Player/Female/Medium/Bodies/F_MED_VelociTeeth/Meshes/Parts/F_MED_VelociTeeth_FaceAcc_Physics
bodies {
 base_scarfFront [{shapeType=Capsule, length=7, radius=5.5, shapeName=spine_05_capsule, center=<-2.8859 -0.0091 1.0702>, rotation=<rot 0 -35 95>}]
 base_tail [{shapeType=Capsule, length=0, radius=1, shapeName=Base_FaceAcc_capsule, center=<1 0 0>, rotation=<rot 90 0 0>}]
 clavicle_l [{shapeType=Capsule, length=6, radius=5.5, shapeName=Base_FaceAcc_capsule, center=<5.25 -2 -1.5>, rotation=<rot 90 0 0>}]
 clavicle_r [{shapeType=Capsule, length=6, radius=5.5, shapeName=Base_FaceAcc_capsule, center=<-5.25 2 1.5>, rotation=<rot 90 0 0>}]
 dyn_braid_L_01 [{shapeType=Capsule, length=1, radius=1, shapeName=dyn_braid_R_01_capsule, center=<2 0 0>, rotation=<rot 90 0 0>}]
 dyn_braid_L_02 [{shapeType=Capsule, length=2, radius=1.5, shapeName=dyn_braid_R_02_capsule, center=<2.5 0 0>, rotation=<rot 90 0 0>}]
 dyn_braid_L_03 [{shapeType=Capsule, length=2, radius=1.5, shapeName=dyn_braid_R_03_capsule, center=<2 0 0>, rotation=<rot 90 0 0>}]
 dyn_braid_L_04 [{shapeType=Capsule, length=2.5, radius=1.5, shapeName=dyn_braid_R_04_capsule, center=<2.25 0 0>, rotation=<rot 90 0 0>}]
 dyn_braid_L_05 [{shapeType=Capsule, length=3.75, radius=1.75, shapeName=dyn_braid_R_04_capsule, center=<2.75 0 0>, rotation=<rot 90 0 0>}]
 dyn_braid_R_01 [{shapeType=Capsule, length=1, radius=1, shapeName=dyn_braid_R_01_capsule, center=<2 0 0>, rotation=<rot 90 180 0>}]
 dyn_braid_R_02 [{shapeType=Capsule, length=2, radius=1.5, shapeName=dyn_braid_R_02_capsule, center=<2.5 0 0>, rotation=<rot 90 180 0>}]
 dyn_braid_R_03 [{shapeType=Capsule, length=2, radius=1.5, shapeName=dyn_braid_R_03_capsule, center=<2 0 0>, rotation=<rot 90 0 0>}]
 dyn_braid_R_04 [{shapeType=Capsule, length=2.5, radius=1.5, shapeName=dyn_braid_R_04_capsule, center=<2.25 0 0>, rotation=<rot 90 0 0>}]
 dyn_braid_R_05 [{shapeType=Capsule, length=3.75, radius=1.75, shapeName=dyn_braid_R_04_capsule, center=<2.75 0 0>, rotation=<rot 90 0 0>}]
 dyn_scarfBack_02 [
  {shapeType=Capsule, length=6, radius=4.25, shapeName=spine_05_capsule, center=<1 4.25 0>}
  {shapeType=Capsule, length=14, radius=2, shapeName=dyn_scarfBack_02_capsule, center=<-2.0249 3.6177 0>}
 ]
 dyn_tail_01 [{shapeType=Capsule, length=1, radius=1.75, shapeName=dyn_tail_01_capsule, center=<2 0 0>, rotation=<rot 90 0 0>}]
 dyn_tail_02 [
  {shapeType=Capsule, length=6.5, radius=2.25, shapeName=dyn_tail_02_capsule, center=<2.25 0.85 0>}
  {shapeType=Box, extentX=1.5, extentY=1.5, extentZ=1, shapeName=dyn_tail_02_box, center=<4.9093 2.5 -5.7454>, rotation=<rot 0 -9 24>}
  {shapeType=Box, extentX=1.5, extentY=1.5, extentZ=1, shapeName=dyn_tail_02_box, center=<4.9093 2.25 5.7454>, rotation=<rot 0 -9 -24>}
 ]
 dyn_tail_03 [{shapeType=Capsule, length=3.5, radius=2, shapeName=dyn_tail_01_capsule, center=<5 -0.5 0>, rotation=<rot 90 -30 0>}]
 head [{shapeType=Capsule, length=3, radius=5, shapeName=base_FaceAcc_jaw_capsule, center=<-1 0 0>, rotation=<rot 90 0 0>}]
 neck_01 [{shapeType=Capsule, length=1, radius=1, shapeName=Base_FaceAcc_capsule, center=<0.5 0 0>, rotation=<rot 90 0 0>}]
 spine_04 [{shapeType=Capsule, length=7, radius=11, shapeName=Base_FaceAcc_capsule, center=<5.2814 -0.804 0>}]
 spine_05 [{shapeType=Capsule, length=1, radius=6, shapeName=Base_FaceAcc_capsule, center=<9.6865 0.1378 0>}]
 upperarm_l [{shapeType=Capsule, length=18, radius=4, shapeName=Base_FaceAcc_capsule, center=<9 0 0>, rotation=<rot 90 0 0>}]
 upperarm_r [
  {shapeType=Capsule, length=24, radius=4, shapeName=Base_FaceAcc_capsule, center=<-12 0 0>, rotation=<rot 90 0 0>}
  {
   shapeType=Box
   extentX=30
   extentY=19.5
   extentZ=5.5
   shapeName=upperarm_r_box
   center=<-18.4702 -2.1814 -3.3764>
   rotation=<rot 0 0.0001 -43.5877>
  }
  {shapeType=Box, extentX=15.5, extentY=3.5, extentZ=12, shapeName=upperarm_r_box, center=<4.715 4 -2.9579>, rotation=<rot 42 0 0>}
 ]
}
constraints [
 {bone1Name=dyn_tail_03, bone2Name=dyn_tail_02, swing1LimitDegrees=20, swing2LimitDegrees=10, twistLimitDegrees=1}
 {bone1Name=dyn_tail_02, bone2Name=dyn_tail_01, swing1LimitDegrees=25, swing2LimitDegrees=8, twistLimitDegrees=1}
 {bone1Name=dyn_tail_01, bone2Name=base_tail, swing1LimitDegrees=15, swing2LimitDegrees=10, twistLimitDegrees=1}
 {bone1Name=dyn_braid_L_02, bone2Name=dyn_braid_L_01, swing1LimitDegrees=15, swing2LimitDegrees=5, twistLimitDegrees=1}
 {bone1Name=dyn_braid_L_03, bone2Name=dyn_braid_L_02, swing1LimitDegrees=25, swing2LimitDegrees=7.5, twistLimitDegrees=1}
 {bone1Name=dyn_braid_L_04, bone2Name=dyn_braid_L_03, swing1LimitDegrees=25, swing2LimitDegrees=12, twistLimitDegrees=1}
 {bone1Name=dyn_braid_L_01, bone2Name=neck_01, swing1LimitDegrees=5, swing2LimitDegrees=5, twistLimitDegrees=1}
 {bone1Name=dyn_braid_R_04, bone2Name=dyn_braid_R_03, swing1LimitDegrees=25, swing2LimitDegrees=15, twistLimitDegrees=1}
 {bone1Name=dyn_braid_R_05, bone2Name=dyn_braid_R_04, swing1LimitDegrees=20, swing2LimitDegrees=15, twistLimitDegrees=1}
 {bone1Name=dyn_braid_R_02, bone2Name=dyn_braid_R_01, swing1LimitDegrees=15, swing2LimitDegrees=10, twistLimitDegrees=1}
 {bone1Name=dyn_braid_L_05, bone2Name=dyn_braid_L_04, swing1LimitDegrees=20, swing2LimitDegrees=15, twistLimitDegrees=1}
 {bone1Name=dyn_braid_R_03, bone2Name=dyn_braid_R_02, swing1LimitDegrees=30, swing2LimitDegrees=10, twistLimitDegrees=1}
 {bone1Name=dyn_braid_R_01, bone2Name=neck_01, swing1LimitDegrees=5, swing2LimitDegrees=5, twistLimitDegrees=1}
 {
  bone1Name=clavicle_l
  bone2Name=spine_05
  swing1LimitDegrees=45
  swing1Motion=Free
  swing2LimitDegrees=45
  swing2Motion=Free
  twistLimitDegrees=45
  twistMotion=Free
 }
 {
  bone1Name=clavicle_r
  bone2Name=spine_05
  swing1LimitDegrees=45
  swing1Motion=Free
  swing2LimitDegrees=45
  swing2Motion=Free
  twistLimitDegrees=45
  twistMotion=Free
 }
 {
  bone1Name=dyn_scarfBack_02
  bone2Name=spine_05
  swing1LimitDegrees=45
  swing1Motion=Free
  swing2LimitDegrees=45
  swing2Motion=Free
  twistLimitDegrees=45
  twistMotion=Free
 }
 {
  bone1Name=head
  bone2Name=spine_05
  swing1LimitDegrees=45
  swing1Motion=Free
  swing2LimitDegrees=45
  swing2Motion=Free
  twistLimitDegrees=45
  twistMotion=Free
 }
 {
  bone1Name=upperarm_l
  bone2Name=spine_05
  swing1LimitDegrees=45
  swing1Motion=Free
  swing2LimitDegrees=45
  swing2Motion=Free
  twistLimitDegrees=45
  twistMotion=Free
 }
 {
  bone1Name=upperarm_r
  bone2Name=spine_05
  swing1LimitDegrees=45
  swing1Motion=Free
  swing2LimitDegrees=45
  swing2Motion=Free
  twistLimitDegrees=45
  twistMotion=Free
 }
]
